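; Kabob placed explosive gadget (kunai mesh) that detonates on a timer, proximity trigger, damage or simulation restart.
; On the server, overlaps explode it with line of sight; otherwise a GameplayEffect applies an LOS checker, and CheckLOSAgain detonates if armed.
; Explode runs once, playing FX, camera shake and sound indicators, then applies the damage effect to LOS-filtered actors in radius.
; Damage from another Kabob delays detonation briefly, staggering chain reactions.
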(event EventBeginPlay
  (|CallParentFunction)
  ;[1 1 1] Caching this so it's not lost when instigator respawns.
  (Variables|Default|SetMyController (Pawn|GetController (Variables|Actor|GetInstigator)))
  (Utilities|Time|SetTimerbyEvent (EventDispatchers|CreateEvent :!event Custom|Explode) (ScalableFloat|GetValueatLevel (Variables|Default|GetExplosionTimer))) ; Setting explosion timer
  (Components|Sphere|SetSphereRadius (Variables|Default|GetExplosionTrigger) (ScalableFloat|GetValueatLevel (Variables|Default|GetTriggerRadius)) true) ; Scale of explosion radius
  (Effects|SpawnEmitteratLocation "/Game/Athena/Items/Weapons/Prototype/Badger_Bangs/FX/P_BlackMonday_HitWorld.P_BlackMonday_HitWorld" (Transformation|GetActorLocation) (Math|Conversions|RotationFromXVector (Variables|Default|GetImpactNormal))) ; Impact VFX
  (CallFunction|PlayAnimMID)
  (EventDispatchers|BindEventtoOnDied (EventDispatchers|CreateEvent :!event Custom|Destroyed)))

(event Custom|Beep
  (params (EventName :type Name) (EmitterTime :type Float%20%28double-precision%29) (Location :type Vector) (Velocity :type Vector))
  (if (Utilities|Name|Equal(Name) EventName "Beep")
    (Audio|PlaySoundatLocation (Variables|Default|GetBeepSound) (Transformation|GetActorLocation))))

(event Custom|Destroyed
  (params (DamagedActor :type Actor%20Object%20Reference) (Damage :type Float%20%28single-precision%29) (InstigatedBy :type Controller%20Object%20Reference) (DamageCauser :type Actor%20Object%20Reference) (HitLocation :type Vector) (FHitComponent :type Primitive%20Component%20Object%20Reference) (BoneName :type Name) (Momentum :type Vector))
  (Audio|Components|Audio|FadeOut (Variables|Default|GetAmbientHumAudio) 0.5 0.0 "Linear" :!class /Script/Engine.AudioComponent)
  (if (not (Variables|Default|GetHasExploded?))
    (Effects|Components|ParticleSystem|SpawnFXAtLocation "/Game/Athena/Items/Weapons/Prototype/Badger_Bangs/FX/NS_BlackMonday_PrjTakeBack.NS_BlackMonday_PrjTakeBack" (Transformation|GetActorLocation) (Math|Conversions|RotationFromXVector (Variables|Default|GetImpactNormal)))
    (Audio|PlaySoundatLocation (Variables|Default|GetPickUpSound) (Transformation|GetActorLocation)))
  (Rendering|SetActorHiddenInGame true)
  (Collision|SetActorEnableCollision false))

(event OnComponentBeginOverlap(ExplosionTrigger)
  (params (OverlappedComponent :type Primitive%20Component%20Object%20Reference) (OtherActor :type Actor%20Object%20Reference) (OtherComp :type Primitive%20Component%20Object%20Reference) (OtherBodyIndex :type Integer) (bFromSweep :type Boolean) (SweepResult :type Hit%20Result%20Structure%20%28by%20ref%29))
  (Networking|SwitchHasAuthority
    (:Authority
      (Utilities|CheckforExplosionTrigger OtherActor
        (:Trigger
          (Utilities|CheckLOS OtherActor
            (:LOS
              (CallFunction|Explode))
            (:No%20LOS
              (GameplayEffects|ApplyGameplayEffectToTarget :self (Ability|GetAbilitySystemComponent self) :GameplayEffectClass "/HeadbandGameplay/Items/Kabob/GE_Kabob_ApplyLOSChecker.GE_Kabob_ApplyLOSChecker_C" :Target (Ability|GetAbilitySystemComponent OtherActor) :Level 0.0))))))))

(event OnComponentEndOverlap(ExplosionTrigger)
  (params (OverlappedComponent :type Primitive%20Component%20Object%20Reference) (OtherActor :type Actor%20Object%20Reference) (OtherComp :type Primitive%20Component%20Object%20Reference) (OtherBodyIndex :type Integer))
  (GameplayEffects|RemoveActiveEffectswithTags (Ability|GetAbilitySystemComponent OtherActor) "(GameplayTags=((TagName=\"Item.Consumable.Kabob\")))"))

(event EventOnDamageServer
  (params (Damage :type Float%20%28single-precision%29) (DamageTags :type Gameplay%20Tag%20Container%20Structure%20%28by%20ref%29) (Momentum :type Vector :default "0, 0, 0") (HitInfo :type Hit%20Result%20Structure%20%28by%20ref%29) (InstigatedBy :type Controller%20Object%20Reference) (DamageCauser :type Actor%20Object%20Reference) (EffectContext :type Gameplay%20Effect%20Context%20Handle%20Structure))
  (Utilities|Casting|CastToBGA_Athena_Headband_Kabob DamageCauser
    (:then
      (Utilities|FlowControl|Delay 0.25))
    (:CastFailed
      _))
  (CallFunction|Explode))

(event Custom|Explode
  (bind _returnvalue (Transformation|GetActorLocation))
  (bind _self self)
  (if (not (Variables|Default|GetHasExploded?))
    (Variables|Default|SetHasExploded? true)
    (Collision|SetActorEnableCollision false)
    (Actor|SetLifeSpan 0.2)
    ;[1 1 0] Stop ambient loop sound indicator
    (SoundIndicator|StopSoundIndicatorLooping (Variables|Default|GetFortSoundIndicator))
    (Audio|PlaySoundatLocation (Variables|Default|GetExplodeSound) _returnvalue "0, 0, 0" 1.0 1.0 0.0 0 0 self) ; Explosion SFX
    (bind _returnvalue_2 (Niagara|SpawnSystematLocation (Variables|Default|GetNE_ExplosionEffect) _returnvalue (Math|Conversions|RotationFromXVector (Variables|Default|GetImpactNormal)))) ; Explosion VFX
    (Niagara|SetNiagaraVariable(Float) _returnvalue_2 "Global Scale" 0.9)
    (GameplayCue|ExecuteGameplayCue%28Burst%29Local self (Variables|Default|GetCameraShakeGC) (Ability|GameplayCue|MakeGameplayCueParameters :NormalizedMagnitude 1.0 :RawMagnitude 3.0 :Location _returnvalue :Normal "0,0,0" :GameplayEffectLevel 1 :AbilityLevel 1 :bReplicateLocationWhenUsingMinimalRepProxy false)) ; Camera Shake Gameplay Cue
    (Fortnite|ShowSoundIndicatorLocalforTeam :Location (Transformation|GetWorldLocation (Variables|Default|GetStaticMesh)) :MaxAudibleDistance (Variables|/Script/Engine.SoundBase|GetMaxDistance (Variables|Default|GetExplodeSound)) :IndicatorType "Gunshot" :Team (Team|GetActorTeam self) :Affiliations (Utilities|Array|MakeArray "Hostile" "Neutral" "Friendly") :Tint "(R=0.000000,G=0.000000,B=0.000000,A=0.000000)" :AttenuationMaxRangeMultiplier 1.0 :AttenuationIntensityMultiplier 1.0) ; Explosion SFX Indicator
    ;[1 1 1] Get actors in range in line of sight
    (bind _outactors (Collision|SphereOverlapActors (Transformation|GetActorLocation) (ScalableFloat|GetValueatLevel (Variables|Default|GetDamageRadius)) (Utilities|Array|MakeArray "ObjectTypeQuery6" "ObjectTypeQuery1" "ObjectTypeQuery5" "ObjectTypeQuery3" "ObjectTypeQuery2" "ObjectTypeQuery4" "ObjectTypeQuery7") "/Script/Engine.Actor" (Utilities|Array|MakeArray self))) ;  
    (bind _result (CallFunction|FilterbyLOS _outactors))
    (for _array_element _result
      (bind _asfort_ability_system_component (Utilities|Casting|CastToFortAbilitySystemComponent (Ability|GetAbilitySystemComponent _self)))
      (GameplayEffects|ApplyGameplayEffectToTarget _asfort_ability_system_component (Variables|Default|GetExplosionGE) (Ability|GetAbilitySystemComponent _array_element) 0.0 (GameplayEffects|MakeFortEffectContext _asfort_ability_system_component 0 0 _self 0)) ; Apply damage GE
    )))

(event EventSpawnFXSounds
  (bind _ambienthumaudio (Variables|Default|GetAmbientHumAudio))
  ;[1 1 1] Play ambient noise
  (|CallParentFunction)
  (Audio|Components|Audio|FadeIn _ambienthumaudio 0.5 1.0 0.0 "Linear" :!class /Script/Engine.AudioComponent)
  ;[1 1 0] Start ambient loop sound indicator
  (SoundIndicator|StartSoundIndicatorLooping (Variables|Default|GetFortSoundIndicator) (Transformation|GetActorLocation) (Variables|/Script/Engine.AudioComponent|GetSound _ambienthumaudio) 1.0 0))

(event EventRestartSimulation
  (CallFunction|Explode))

(fn BlueprintCanInteract
  (params (InteractingPawn :type Fort%20Pawn%20Object%20Reference) (InteractionBeingAttempted :type EInteractionBeingAttempted%20Enum :default "FirstInteraction") (InteractionType :type TInteractionType%20Enum :default "IT_NoInteraction"))
  (return false))

(fn BlueprintGetInteractionString
  (params (InteractingPawn :type Fort%20Pawn%20Object%20Reference) (InteractionBeingAttempted :type EInteractionBeingAttempted%20Enum :default "FirstInteraction"))
  (return (Variables|Default|GetInteractText)))

(fn BlueprintGetInteractionTime
  (params (InteractingPawn :type Fort%20Pawn%20Object%20Reference) (InteractionBeingAttempted :type EInteractionBeingAttempted%20Enum :default "FirstInteraction"))
  (return 0.0 true))

(fn CheckLOSAgain
  (params (ActorToCheckLOSAgainst :type Actor%20Object%20Reference))
  (Networking|SwitchHasAuthority
    (:Authority
      (Utilities|CheckLOS ActorToCheckLOSAgainst
        (:LOS
          (if (Variables|Default|GetArmed)
            (CallFunction|Explode)))))))

(fn FilterByLOS
  (params (_Array :type Array%20of%20Actor%20Object%20References))
  (for _array_element _Array
    (Utilities|CheckLOS _array_element
      (:LOS
        (Utilities|Array|Add (LocalVariables|GetResult) _array_element))))
  (return (LocalVariables|GetResult)))

(fn GetAffiliationsToShowFor
  (return (Utilities|Array|MakeArray "Hostile" "Neutral" "Friendly")))

(fn Launch
  (params (LaunchSource :type Vector) (Amount :type Float%20%28double-precision%29))
  (bind _returnvalue (Transformation|GetActorLocation))
  (bind _returnvalue_2 (- _returnvalue LaunchSource))
  (bind _gravhitnormal (Variables|Default|GetGravHitNormal))
  (bind _returnvalue_3 (* _gravhitnormal Amount))
  (bind _returnvalue_4 (Math|Vector|Normalize _returnvalue_2))
  (bind _returnvalue_5 (* _returnvalue_4 Amount))
  (bind _returnvalue_6 (+ _returnvalue_5 _returnvalue_3))
  (bind _returnvalue_7 (Math|Vector|Normalize _returnvalue_6))
  (bind _returnvalue_8 (* _returnvalue_7 1e+02))
  (bind _returnvalue_9 (Transformation|GetActorLocation))
  (bind _returnvalue_10 (+ _returnvalue_9 _returnvalue_8))
  (CallFunction|RestartSimulation)
  (Transformation|DetachFromActor "KeepWorld" "KeepWorld" "KeepWorld")
  (bind (_outhit _returnvalue_11) (Collision|LineTraceByChannel _returnvalue_9 _returnvalue_10 "TraceTypeQuery10"))
  (Development|PrintString (Utilities|String|ToString(Vector) _returnvalue_8) true true "(R=1.000000,G=0.000000,B=0.847731,A=1.000000)" 8.0)
  (Transformation|SetActorLocation (Math|Vector|SelectVector (.Location (Collision|BreakHitResult _outhit)) _returnvalue_10 _returnvalue_11) false true)
  (Variables|/Script/Engine.MovementComponent|SetVelocity _returnvalue_6 (Variables|Default|GetProjectileMovement)))

(fn PlayAnimMID
  (bind _returnvalue (Rendering|Material|CreateDynamicMaterialInstance (Variables|Default|GetSK_Kunai) 0 "/HeadbandGameplay/Items/Kabob/Materials/MI_Gadget_Headband_Kabob.MI_Gadget_Headband_Kabob" "None" :!class /Script/Engine.PrimitiveComponent))
  (Rendering|Material|SetScalarParameterValue _returnvalue "UseAnimation" 1.0 :!class /Script/Engine.MaterialInstanceDynamic)
  (Rendering|Material|SetScalarParameterValue _returnvalue "AnimStart" (Utilities|Time|GetGameTimeinSeconds) :!class /Script/Engine.MaterialInstanceDynamic))

(fn ShouldShowSoundIndicator
  (params (PlayerController :type Fort%20Player%20Controller%20Object%20Reference))
  (return true))

(fn ConstructionScript
  (Networking|SwitchHasAuthority
    (:Authority
      (Variables|/Script/Engine.PrimitiveComponent|SetbGenerateOverlapEvents true (Variables|Default|GetExplosionTrigger))
      (|CallParentFunction))))
