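; Thrown Kabob (kunai) projectile for the Headband item, which explodes on valid targets or sticks into the world.
; Homing locks onto targets within a forward cone of the thrower and stops on a timer.
; Proximity, capsule and water events trigger a one-shot explosion, with water flagged for explosion selection.
; World hits make the server spawn an embedded gameplay actor aligned to the surface, while hitting another Kabob raises gravity.
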(event EventBeginPlay
  ;[1 1 1] Begin Play.  Return if nothing is hit after a delay
  (|CallParentFunction)
  (Variables|Default|SetisDead false)
  (Actor|SetLifeSpan 3e+01)
  (Utilities|FlowControl|Sequence
    (:then_0
      (bind _asfort_player_pawn (Utilities|Casting|CastToFortPlayerPawn (Variables|Actor|GetInstigator)))
      (bind _output_get (Variables|Default|SetOwningPlayerPawn _asfort_player_pawn))
      (Audio|PlaySoundatLocation (select (Pawn|IsLocallyControlled _output_get) (Variables|Default|GetTrueSoundCue) (Variables|Default|GetFalseSoundCue)) (Transformation|GetActorLocation)))
    (:then_1
      (Utilities|Time|SetTimerbyEvent (EventDispatchers|CreateEvent :!event Custom|StopHoming) (ScalableFloat|GetValueatLevel (Variables|Default|GetHomingDuration)))
      (Utilities|Array|AddUnique (Variables|Default|GetActorsToIgnoreForTargeting) (Variables|Default|GetOwningPlayerPawn))
      (Components|Sphere|SetSphereRadius (Variables|Default|GetProximityDetection) (ScalableFloat|GetValueatLevel (Variables|Default|GetProxTriggerRadius)) true))))

(event Custom|StopHoming
  ;[1 1 1] Begin Play.  Return if nothing is hit after a delay
  (Projectile|SetHomingStyle (Variables|Movement|GetProjectileMovementComponent) "None" false :!class /Script/FortniteGame.FortProjectileMovementComponent))

(event OnComponentActivated(ProjectileMovementComponent)
  (params (Component :type Actor%20Component%20Object%20Reference) (bReset :type Boolean))
  ;[1 1 1] AnimTrail
  (Effects|Components|ParticleSystem|BeginLegacyAnimTrails (Variables|Default|GetNS_Kabob_Trail)))

(event Custom|SpawnBGA
  (params (HitResult :type Hit%20Result%20Structure%20%28by%20ref%29))
  (bind (_bblockinghit _binitialoverlap _time _distance _location _impactpoint) (Collision|BreakHitResult HitResult))
  (bind (_bblockinghit_2 _binitialoverlap_2 _time_2 _distance_2 _location_2 _impactpoint_2 _normal _impactnormal _physmat _hitactor _hitcomponent) (Collision|BreakHitResult HitResult))
  ;[1 1 1] Spawn BGA
  (Utilities|FlowControl|DoOnce false
    (:Completed
      (Utilities|FlowControl|Sequence
        (:then_0
          (Actor|SetLifeSpan 0.25)
          (Collision|SetActorEnableCollision false)
          (Rendering|SetActorHiddenInGame true))
        (:then_1
          (bind _returnvalue (Building|SpawnBuildingGameplayActor (Variables|Default|GetBGAtoSpawn) (Math|Transform|MakeTransform (- _impactpoint (* (Transformation|GetForwardVector (Variables|B_Prj_Grenade|GetMesh)) (Variables|Default|GetKunaiBGA_PiercingDepthOffset))) (CallFunction|GetBGARotation HitResult) "2.000000,2.000000,2.000000") (Variables|Actor|GetInstigator) false false))
          (bind _asbga_athena_with_gravity_parent (Utilities|Casting|CastToBGA_Athena_WithGravity_Parent _returnvalue))
          (Class|BGAAthenawithGravityParent|MulticastHandleBinding _asbga_athena_with_gravity_parent _hitactor _hitcomponent)
          (Class|BGAAthenawithGravityParent|Init _asbga_athena_with_gravity_parent _impactnormal)
          (bind _asbga_athena_headband_kabob (Utilities|Casting|CastToBGA_Athena_Headband_Kabob _asbga_athena_with_gravity_parent))
          (Variables|/HeadbandGameplay/Items/Kabob/BGA_Athena_Headband_Kabob.BGA_Athena_Headband_Kabob_C|SetImpactNormal _impactnormal _asbga_athena_headband_kabob))))))

(event Custom|OnHitSomething
  (params (InputPin :type Hit%20Result%20Structure%20%28by%20ref%29))
  (bind _lasthit (Variables|Default|GetLastHit))
  (bind (_bblockinghit _binitialoverlap _time _distance _location _impactpoint _normal) (Collision|BreakHitResult _lasthit))
  (bind _projectilemovementcomponent (Variables|Movement|GetProjectileMovementComponent))
  ;[1 1 1] Event Hit: Trigger a Projectile SphereTrace.
  (if (Variables|Default|GetHasHitWater)
    (CallFunction|HitExplosionTrigger 0)
    (else
      (Variables|Default|SetLastHit InputPin)
      (Effects|Components|ParticleSystem|EndLegacyAnimTrails (Variables|Default|GetNS_Kabob_Trail))
      (Variables|Default|SetLastVelocity (Variables|/Script/Engine.MovementComponent|GetVelocity _projectilemovementcomponent))
      (Variables|/Script/Engine.MovementComponent|SetVelocity "0, 0, 0" _projectilemovementcomponent)
      (Utilities|Time|PauseTimerbyHandle (Variables|Default|GetReturnTimer))
      (Networking|SwitchHasAuthority
        (:Authority
          (Variables|Default|SetProjectileHitNormal _normal)
          (Variables|Default|SetProjectileHitLoc _location)
          (Projectile|SetHomingTarget 0 self 0 0 :!class /Script/FortniteGame.FortProjectileMovementInterface)
          (CallFunction|HitWorld (Variables|Default|GetHitEnvSoundCue) (Variables|Default|GetProjectileHitLoc)))))))

(event OnComponentBeginOverlap(CapsuleComponent)
  (params (OverlappedComponent :type Primitive%20Component%20Object%20Reference) (OtherActor :type Actor%20Object%20Reference) (OtherComp :type Primitive%20Component%20Object%20Reference) (OtherBodyIndex :type Integer) (bFromSweep :type Boolean) (SweepResult :type Hit%20Result%20Structure%20%28by%20ref%29))
  (bind (_asfort_water_body_actor _bsuccess) (Utilities|Casting|CastToFortWaterBodyActor OtherActor))
  (bind (_asfort_water_body_bp _bsuccess_2) (Utilities|Casting|CastToFortWaterBodyBP OtherActor))
  (bind (_asfort_athena_water _bsuccess_3) (Utilities|Casting|CastToFortAthenaWater OtherActor))
  ;[1 1 1] Event Hit: Trigger a Projectile SphereTrace.
  (if (Math|Boolean|ORBoolean _bsuccess_2 _bsuccess_3 _bsuccess)
    (Variables|Default|SetHasTouchedWater true)
    (CallFunction|HitExplosionTrigger OtherActor)))

(event OnEnterWaterEvent(WaterInteractionComponent)
  (params (WaterBody :type Fort%20Water%20Body%20Actor%20Object%20Reference) (WaterInteractionComponent :type Fort%20Water%20Interaction%20Component%20Object%20Reference) (bIsFirstBody :type Boolean))
  ;[1 1 1] Event Hit: Trigger a Projectile SphereTrace.
  (Variables|Default|SetHasTouchedWater true)
  (CallFunction|HitExplosionTrigger WaterBody))

(event Custom|HitExplosionTrigger
  (params (HitActor :type Actor%20Object%20Reference))
  ;[1 1 1] Event Hit: Trigger a Projectile SphereTrace.
  (CallFunction|PlayHitPawnSounds HitActor)
  (Utilities|FlowControl|DoOnce false
    (:Completed
      (Delivery|DoExplosion)
      (GameplayCue|ExecuteGameplayCue%28Burst%29Local self (Variables|Default|GetCameraShakeGC) (Ability|GameplayCue|MakeGameplayCueParameters :NormalizedMagnitude 1.0 :RawMagnitude 3.0 :Location (Transformation|GetActorLocation) :Normal "0,0,0" :Instigator (Variables|Actor|GetInstigator) :EffectCauser self :GameplayEffectLevel 1 :AbilityLevel 1 :bReplicateLocationWhenUsingMinimalRepProxy false))
      (Rendering|SetActorHiddenInGame true)
      (Collision|SetActorEnableCollision false)
      (Actor|SetLifeSpan 0.5)
      (Variables|Default|SetisDead true))))

(event OnComponentHit(CapsuleComponent)
  (params (HitComponent :type Primitive%20Component%20Object%20Reference) (OtherActor :type Actor%20Object%20Reference) (OtherComp :type Primitive%20Component%20Object%20Reference) (NormalImpulse :type Vector :default "0, 0, 0") (Hit :type Hit%20Result%20Structure%20%28by%20ref%29))
  (bind (_bblockinghit _binitialoverlap _time _distance _location _impactpoint _normal) (Collision|BreakHitResult Hit))
  ;[1 1 1] Event Hit: Trigger a Projectile SphereTrace.
  (if (not (Variables|Default|GetisDead))
    (Utilities|Casting|CastToBGA_Athena_Headband_Kabob OtherActor
      (:then
        (Effects|SpawnEmitteratLocation "/Game/Athena/Items/Weapons/Prototype/Badger_Bangs/FX/P_BlackMonday_HitWorld.P_BlackMonday_HitWorld" _location (Math|Conversions|RotationFromXVector _normal))
        (Variables|/Script/Engine.ProjectileMovementComponent|SetProjectileGravityScale 5.0 (Variables|Movement|GetProjectileMovementComponent)))
      (:CastFailed
        (Utilities|CheckforExplosionTriggers OtherActor
          (:Triggered
            (CallFunction|HitExplosionTrigger OtherActor))
          (:Not%20Triggered
            (CallFunction|OnHitSomething Hit)))))))

(event Custom|HitWorld
  (params (Sound :type Sound%20Base%20Object%20Reference) (Location :type Vector))
  ;[1 1 1] Event Hit: Trigger a Projectile SphereTrace.
  (CallFunction|SpawnBGA (Variables|Default|GetLastHit))
  (Audio|PlaySoundatLocation Sound Location))

(event EventOnExploded
  (params (HitActors :type Array%20of%20Actor%20Object%20References) (HitResults :type Array%20of%20Hit%20Result%20Structures))
  ;[1 1 1] Event Hit: Trigger a Projectile SphereTrace.
  (|CallParentFunction HitActors HitResults)
  (Fortnite|ShowSoundIndicatorLocalforTeam :Location (Transformation|GetWorldLocation (Variables|B_Prj_Grenade|GetMesh)) :MaxAudibleDistance (Variables|/Script/Engine.SoundBase|GetMaxDistance (Variables|Sounds|GetCue_ExplosionSound)) :IndicatorType "Gunshot" :Team (Team|GetActorTeam self) :Affiliations (Utilities|Array|MakeArray "Hostile" "Neutral" "Friendly") :Tint "(R=0.000000,G=0.000000,B=0.000000,A=0.000000)" :AttenuationMaxRangeMultiplier 1.0 :AttenuationIntensityMultiplier 1.0) ; Explosion SFX Indicator
)

(event OnComponentBeginOverlap(ProximityDetection)
  (params (OverlappedComponent :type Primitive%20Component%20Object%20Reference) (OtherActor :type Actor%20Object%20Reference) (OtherComp :type Primitive%20Component%20Object%20Reference) (OtherBodyIndex :type Integer) (bFromSweep :type Boolean) (SweepResult :type Hit%20Result%20Structure%20%28by%20ref%29))
  ;[1 1 1] Event Hit: Trigger a Projectile SphereTrace.
  (if (Math|Boolean|ANDBoolean (Variables|Default|GetExplodeOnImpactPlayer) (> (Actor|GetGameTimeSinceCreation) 0.025) (not (Variables|Default|GetisDead)))
    (Utilities|CheckforExplosionTriggers OtherActor
      (:Triggered
        (Utilities|IsPlayerinShoppingCart? OtherActor
          (:True
            _)
          (:False
            (Utilities|CheckLOS OtherActor
              (:LOS
                _))))
        (CallFunction|HitExplosionTrigger OtherActor)))))

(fn GetBGARotation
  (params (Hit :type Hit%20Result%20Structure%20%28by%20ref%29))
  (bind (_bblockinghit _binitialoverlap _time _distance _location _impactpoint _normal _impactnormal _physmat _hitactor _hitcomponent _hitbonename _bonename _hititem _elementindex _faceindex _tracestart _traceend) (Collision|BreakHitResult Hit))
  (return (Math|Rotator|MakeRotator (.roll (Math|Rotator|MakeRotfromZX (Math|Vector|SelectVector _impactnormal "0, 0,1.000000" (Math|Vector|NotEqual(Vector) _impactnormal "0, 0, 0" 0.0001)) (Math|Vector|ProjectVectorontoPlane (- _traceend _tracestart) (Math|Vector|SelectVector _impactnormal "0, 0,1.000000" (Math|Vector|NotEqual(Vector) _impactnormal "0, 0, 0" 0.0001))))) (+ (.pitch (Math|Rotator|MakeRotfromZX (Math|Vector|SelectVector _impactnormal "0, 0,1.000000" (Math|Vector|NotEqual(Vector) _impactnormal "0, 0, 0" 0.0001)) (Math|Vector|ProjectVectorontoPlane (- _traceend _tracestart) (Math|Vector|SelectVector _impactnormal "0, 0,1.000000" (Math|Vector|NotEqual(Vector) _impactnormal "0, 0, 0" 0.0001))))) (Math|Random|RandomFloatinRange -8.0 8.0)) (.yaw (Math|Rotator|MakeRotfromZX (Math|Vector|SelectVector _impactnormal "0, 0,1.000000" (Math|Vector|NotEqual(Vector) _impactnormal "0, 0, 0" 0.0001)) (Math|Vector|ProjectVectorontoPlane (- _traceend _tracestart) (Math|Vector|SelectVector _impactnormal "0, 0,1.000000" (Math|Vector|NotEqual(Vector) _impactnormal "0, 0, 0" 0.0001))))))))

(fn GetExplosion
  (params (InWater :type Boolean))
  (return (|CallParentFunction (or InWater (Variables|Default|GetHasTouchedWater)))))

(fn OffsetBGA_Rotation
  (params (InRot :type Rotator))
  (return (Math|Rotator|MakeRotator (.roll InRot) (+ (.pitch InRot) (Math|Random|RandomFloatinRange -45.0 -3e+01)) (.yaw InRot))))

(fn OnRep_bHit)

(fn OnRep_HitLocation)

(fn OnRep_PlayerLocation)

(fn OnRep_Target)

(fn PlayHitPawnSounds
  (params (HitActor :type Actor%20Object%20Reference))
  (bind _returnvalue (Transformation|GetActorLocation))
  (bind _owningplayerpawn (Variables|Default|GetOwningPlayerPawn))
  (Utilities|IsValid _owningplayerpawn
    (:Is%20Valid
      (if (Pawn|IsLocallyControlled _owningplayerpawn)
        (Audio|PlaySoundatLocation (Variables|Default|GetHitPlayerSoundCue) _returnvalue)
        (else
          (bind _asfort_player_pawn (Utilities|Casting|CastToFortPlayerPawn HitActor)
            (:then
              (if (Pawn|IsLocallyControlled _asfort_player_pawn)
                (Audio|PlaySoundatLocation (Variables|Default|GetHitPlayerSoundCue) _returnvalue)
                (else
                  _)))
            (:CastFailed
              _))
          (Audio|PlaySoundatLocation (Variables|Default|GetHitEnvSoundCue) _returnvalue))))))

(fn PrintActorsFromArray
  (params (ActorArray :type Array%20of%20Actor%20Object%20References))
  (for _ ActorArray
    (Development|PrintString (Utilities|String|Append "Hit;" (Utilities|GetDisplayName (Utilities|Array|Get(acopy) ActorArray (LocalVariables|GetHitActorIndex)))) true true "(R=1.000000,G=0.019947,B=0.059748,A=1.000000)" 5.0)
    (LocalVariables|SetHitActorIndex (+ (LocalVariables|GetHitActorIndex) 1))))

(fn SetHomingVars
  (params (HomingTarget :type Actor%20Object%20Reference))
  (bind _self self)
  (bind _homingtarget (Variables|Default|GetHomingTarget))
  (if (Variables|Default|GetHoming)
    (Utilities|ShouldHomingIgnoreDBNO HomingTarget
      (:OutputPin
        (bind _output_get (Variables|Default|SetHomingTarget HomingTarget))
        (Utilities|IsValid _output_get
          (:Is%20Valid
            (Projectile|SetHomingStyle 0 0 "LockOn" false :!class /Script/FortniteGame.FortProjectileMovementInterface)
            (Utilities|Array|AddUnique (Variables|Default|GetActorsToIgnoreForTargeting))
            (Networking|SwitchHasAuthority
              (:Authority
                ;[1 1 1] TODO: Make hotfixable values.
                (Projectile|SetHomingTurnSpeedMin 0 _self 1e+03 :!class /Script/FortniteGame.FortProjectileMovementInterface)
                (Projectile|SetHomingTurnSpeedMax 0 _self 1e+03 :!class /Script/FortniteGame.FortProjectileMovementInterface)
                (Projectile|SetHomingTimeUntilMaxTurnSpeed 0 _self 0.5 :!class /Script/FortniteGame.FortProjectileMovementInterface)
                (Projectile|SetHomingTargetOffset _self (Variables|Default|GetTargetHomingOffset) :!class /Script/FortniteGame.FortProjectileMovementInterface)
                (Utilities|IsValid _homingtarget
                  (:Is%20Valid
                    (if (>= (Math|Vector|DotProduct (Math|Vector|Normalize (- (Transformation|GetActorLocation (Variables|Default|GetHomingTarget)) (Transformation|GetActorLocation (Variables|Default|GetOwningPlayerPawn)))) (Math|Vector|Normalize (Transformation|GetActorForwardVector (Variables|Default|GetOwningPlayerPawn)))) 0.49)
                      (Projectile|SetHomingTarget 0 self _homingtarget 0 :!class /Script/FortniteGame.FortProjectileMovementInterface))))))))))))

(fn ConstructionScript
  (|CallParentFunction))
