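; Gameplay ability that applies the explosion effects of the Kabob headband item from its BGA_Athena_Headband_Kabob avatar actor.
; On activation it stores a source transform at the avatar's location, nudged out along the Kabob's impact normal, and returns it through the GetCustomAbilitySourceTransform override.
; It then runs target selection with the generic projectile Explode tag, applies the gameplay effect container to the targets and ends.
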(event Ability|EventActivateAbility
  (bind _returnvalue (Ability|GetAvatarActorfromActorInfo))
  (bind (_returnvalue_location _returnvalue_rotation _returnvalue_scale) (Math|Transform|BreakTransform (Transformation|GetActorTransform _returnvalue)))
  (bind _asbga_athena_headband_kabob (Utilities|Casting|CastToBGA_Athena_Headband_Kabob _returnvalue))
  (Variables|SetSource_Location (Math|Transform|MakeTransform (+ (* (Variables|/HeadbandGameplay/Items/Kabob/BGA_Athena_Headband_Kabob.BGA_Athena_Headband_Kabob_C|GetImpactNormal _asbga_athena_headband_kabob) 3e+01) _returnvalue_location) _returnvalue_rotation _returnvalue_scale))
  (bind (_asynctaskproxy _targetdata _applicationtag) (Ability|Tasks|DoTargetSelection "None" "(TagName=\"Abilities.Generic.GenericProjectile.Explode\")" true)
    (:Targeted
      (Ability|ApplyGameplayEffectContainer _targetdata _applicationtag)
      (Ability|EndAbility))))

(fn GetCustomAbilitySourceTransform
  (return (Variables|Default|GetSource)))
