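; Gameplay ability for the Headband Kabob item that repeatedly requests a line-of-sight recheck on its avatar while active.
; Activation starts a short looping timer, and ending the ability pauses that timer rather than clearing it.
; Each tick casts the effect causer from the granting effect's context to BGA_Athena_Headband_Kabob and calls its CheckLOSAgain with the avatar, leaving the check itself to that actor.
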(event Ability|EventActivateAbility
  (bind _returnvalue (Utilities|Time|SetTimerbyEvent (delegate Custom|CheckLOSAgain) 0.25 true)))

(event Ability|EventOnEndAbility
  (params (bWasCancelled :type Boolean))
  (Utilities|Time|PauseTimerbyHandle (stale Ability|EventActivateAbility _returnvalue)))

(event Custom|CheckLOSAgain
  (bind (_returnvalue_owneractor _returnvalue_avataractor) (Utilities|Struct|BreakGameplayAbilityActorInfo (Ability|GetActorInfo)))
  (bind _asbga_athena_headband_kabob (Utilities|Casting|CastToBGA_Athena_Headband_Kabob (Ability|EffectContext|GetEffectCauser (Ability|GetGrantedbyEffectContext))))
  (Class|BGAAthenaHeadbandKabob|CheckLOSAgain _asbga_athena_headband_kabob _returnvalue_avataractor))
