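# Masked master material for the Kabob headband gadget, adding an animated flaming dissolve and wind motion to anime character shading.
# Time from AnimStart over Duration, or a FakeAnim override, drives a noise-broken threshold across a pre-skinned gradient for opacity and glowing edge.
# Panned, distorted flame texture adds hue-shifted emissive within a height mask, gated with opacity by UseAnimation.
# Panned noise drives World Position Offset, weighted by a height gradient.
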Material /HeadbandGameplay/Items/Kabob/Materials/M_Gadget_Headband_Kabob_Master.M_Gadget_Headband_Kabob_MASTER
graph {
 properties {
  bUseMaterialAttributes=true
  bUsedWithMeshDeformer=true
  bUsedWithSkeletalMesh=true
  blendMode=BLEND_Masked
  pixelDepthOffsetMode=PDOM_Legacy
  usedShadingModels=Default Lit
 }
 nodes [
  n1 = MaterialFunctionCall() {materialFunction=&BaseFN_CharacterFunction_PreFX}
  n2 = MaterialFunctionCall!(
   Material=@n1.Result
   LitDiffuseOverride
   SSC_AlphaOverride
   Tint Mask
   DF InkLine Visibility Mask
   DF Inkline Injection
   DF Inkline Color Override
   DF Inkline Color Override Mask
   CustomUVs
  ) {materialFunction=/Game/Packages/Fortress_SharedMaterials/Anime_Master_Material/MF_AnimeShading}
  MaterialFunctionCall(AO=@n1.AO, Distance Blend=@n1.AO, Fresnel=@n1.AO, Gmap=@n1.AO, Pre Skinned Local Position=@n1.AO, Material=@n2.Result) {materialFunction=&BaseFN_CharacterFunction_PostFX}
  n3 = PreSkinnedPosition()
  n4 = SphereMask(A=@n3, B=Constant3Vector() {constant=<rgba 0 0 -25 0>}, Radius=Constant() {r=34}, Hardness=Constant())
  NamedRerouteDeclaration(None=VertexInterpolator(VS=Saturate(None=@n4)).PS) {name=Gradient, nodeColor=<rgba 0 1 0.9063 1>}
  n5 = Divide(A=ComponentMask(None=@n3) {b=true}) {constB=62}
  n6 = Subtract(A=Multiply(A=Add(A=@n5)) {constB=1.6}) {constB=0.95}
  n7 = Sign(None=Saturate(None=@n6))
  NamedRerouteDeclaration(None=VertexInterpolator(VS=OneMinus(None=@n7)).PS) {name=FlameMask, nodeColor=<rgba 1 0.7031 0 1>}
  NamedRerouteDeclaration(None=Saturate(None=Multiply(A=@n6) {constB=-1})) {name=WindGradientWPO, nodeColor=<rgba 1 0.7031 0 1>}
  n8 = GetMaterialAttributes(None=@n2.Result) {attributeGetTypes [#1, #2]}
  n9 = Time() {period=20}
  n10 = ScalarParameter() {defaultValue=-10, parameterName=AnimStart}
  n11 = ScalarParameter() {defaultValue=3.5, parameterName=Duration}
  n12 = ScalarParameter() {parameterName=FakeAnim, sliderMax=1, sliderMin=0.2762}
  n13 = LinearInterpolate(A=Multiply(A=Subtract(A=@n9, B=@n10), B=Divide(B=@n11)) {constB=0.2}, B=@n12, Alpha=Sign(None=@n12))
  n14 = Subtract(
   A=ConstantBiasScale(None=Saturate(None=@n13)) {bias=-0.5, scale=2}
   B=Multiply(A=TextureSample() {samplerType=SAMPLERTYPE_LinearColor, texture=/Game/Global/Textures/Default/Caustics/03_WaterFoam_AccentSwirls_H}.R) {constB=0.05}
  )
  n15 = SmoothStep(Min=@n14, Max=Add(A=@n14) {constB=0.1}, Value=NamedRerouteUsage())
  n16 = Multiply(A=OneMinus(None=@n15), B=Constant3Vector() {constant=<rgba 5 0.9489 0 0>})
  n17 = NamedRerouteUsage()
  n18 = Panner(Coordinate=Multiply(A=TextureCoordinate()) {constB=0.25}) {speedX=-0.03, speedY=-0.06}
  n19 = Time()
  n20 = Sine(None=Add(
    A=TextureSample(UVs=@n18) {samplerType=SAMPLERTYPE_LinearColor, texture=&03_WaterFoam_AccentSwirls_H}.R
    B=Multiply(A=@n19) {constB=0.2}
   ))
  n21 = TextureSample(UVs=Add(A=Panner() {speedX=0.1, speedY=1.2}, B=Multiply(A=@n20) {constB=0.005})) {samplerType=SAMPLERTYPE_LinearGrayscale, texture=/Game/Weapons/FORT_Melee/Pickaxe_Cube_Ninja_Male/FX/T_FluidFlameTile}
  n22 = ScalarParameter() {defaultValue=1}
  n23 = SmoothStep(Min=Add(A=@n21.R, B=@n22) {constB=0.5}, Max=Subtract(A=@n21.R, B=@n22), Value=Subtract(A=NamedRerouteUsage(), B=@n14))
  n24 = VectorParameter() {parameterName=FlameColor, defaultValue=<rgba 1 0 0.0102 0>}
  n25 = Multiply(
   A=@n17
   B=MaterialFunctionCall(Hue Shift Percentage=Divide(A=@n23) {constB=5}, Texture=Multiply(A=@n23, B=@n24.RGB)) {materialFunction=&HueShift}.Result
  )
  n26 = Multiply(A=ScalarParameter() {parameterName=UseAnimation}, B=NamedRerouteUsage())
  n27 = Multiply(A=@n23, B=@n17)
  n28 = Constant3Vector() {constant=<rgba 1 3.5 1 0>}
  n29 = Divide(A=ComponentMask(None=PreSkinnedPosition()) {b=true, r=true}, B=ScalarParameter() {defaultValue=300, parameterName=Param_2})
  n30 = ScalarParameter() {defaultValue=0.2, parameterName=Param_1}
  n31 = ConstantBiasScale(None=TextureSample(UVs=Panner(Coordinate=@n29, Speed=@n30) {speedX=0.5, speedY=1}) {samplerType=SAMPLERTYPE_LinearColor, texture=/Game/Effects/Fort_Effects/BluGlow/Violentium/Violentium_Mask}.R) {bias=-0.5, scale=1}
  n32 = NamedRerouteUsage()
  SetMaterialAttributes!(
   MaterialAttributes=@n8.MaterialAttributes
   Emissive Color=LinearInterpolate(A=@n8.Emissive Color, B=Add(A=@n16, B=@n25), Alpha=Multiply(A=@n26, B=@n27))
   Opacity Mask=LinearInterpolate(B=@n15, Alpha=@n26) {constA=1}
   World Position Offset=Multiply(A=Multiply(A=@n28, B=@n31), B=Power(Base=@n32) {constExponent=0.5})
  ) {attributeSetTypes [#1, #3, #4]}
  TextureCoordinate()
  VertexInterpolator(VS=@n3)
 ]
}
