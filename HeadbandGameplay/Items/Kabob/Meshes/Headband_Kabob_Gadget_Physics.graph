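# Physics asset for the Kabob headband gadget mesh, simulating a dangling tag on a short string.
# A master anchor and three string segments use capsules, and four tag segments use flat boxes.
# Constraints link them in one serial chain from the anchor through the string to the last tag segment.
# Swing and twist limits are looser along the string and its tag joint, and tighter between tag segments, making the tag stiffer.
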PhysicsAsset /HeadbandGameplay/Items/Kabob/Meshes/Headband_Kabob_Gadget_Physics
bodies {
 dyn_string_01 [{shapeType=Capsule, length=1, radius=0.5, shapeName=dyn_string_01_capsule, center=<1 0 0>, rotation=<rot 90 0 0>}]
 dyn_string_02 [{shapeType=Capsule, length=4, radius=0.5, shapeName=dyn_string_02_capsule, center=<2.5 0 0>, rotation=<rot 90 0 0>}]
 dyn_string_03 [{shapeType=Capsule, length=4, radius=0.5, shapeName=dyn_string_03_capsule, center=<2.5 0 0>, rotation=<rot 90 0 0>}]
 dyn_string_mstr [{shapeType=Capsule, length=0.505, radius=0.505, shapeName=dyn_string_mstr_capsule}]
 dyn_tag_01 [{shapeType=Box, extentX=7, extentY=2, extentZ=20, shapeName=dyn_tag_01_box, center=<3.5 0 0>}]
 dyn_tag_02 [{shapeType=Box, extentX=7, extentY=2, extentZ=20, shapeName=dyn_tag_02_box, center=<3.5 0 0>}]
 dyn_tag_03 [{shapeType=Box, extentX=7, extentY=2, extentZ=20, shapeName=dyn_tag_03_box, center=<3.5 0 0>}]
 dyn_tag_04 [{shapeType=Box, extentX=7, extentY=2, extentZ=20, shapeName=dyn_tag_04_box, center=<3.5 0 0>}]
}
constraints [
 {bone1Name=dyn_string_01, bone2Name=dyn_string_mstr, swing1LimitDegrees=45, swing2LimitDegrees=45, twistLimitDegrees=15}
 {bone1Name=dyn_string_02, bone2Name=dyn_string_01, swing1LimitDegrees=45, swing2LimitDegrees=45, twistLimitDegrees=15}
 {bone1Name=dyn_string_03, bone2Name=dyn_string_02, swing1LimitDegrees=45, swing2LimitDegrees=45, twistLimitDegrees=15}
 {bone1Name=dyn_tag_01, bone2Name=dyn_string_03, swing1LimitDegrees=45, swing2LimitDegrees=45, twistLimitDegrees=15}
 {bone1Name=dyn_tag_02, bone2Name=dyn_tag_01, swing1LimitDegrees=30, swing2LimitDegrees=30, twistLimitDegrees=5}
 {bone1Name=dyn_tag_03, bone2Name=dyn_tag_02, swing1LimitDegrees=30, swing2LimitDegrees=30, twistLimitDegrees=5}
 {bone1Name=dyn_tag_04, bone2Name=dyn_tag_03, swing1LimitDegrees=30, swing2LimitDegrees=30, twistLimitDegrees=5}
]
